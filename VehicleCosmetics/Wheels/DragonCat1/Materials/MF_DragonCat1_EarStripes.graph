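# Material function for the DragonCat1 wheel cosmetic that draws animated three-color ear stripes, output as BaseColor and scaled Emissive material attributes.
# Polar UVs around a pulsing center on UV1 scroll in angle, pass through a sine and sample the EarAni mask at two slight rotations.
# These samples, plus time-driven waves from the Eff_Ani blue channel, blend the colors.
# A second Eff_Ani wave becomes a hard mask that cuts black bands.
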MaterialFunction /VehicleCosmetics/Wheels/DragonCat1/Materials/MF_DragonCat1_EarStripes
graph {nodes [
  n1 = VectorParameter() {group=9_3_EarStripes, parameterName=Ear_Color2, defaultValue=<rgba 0 0.8073 0.7824 1>}
  n2 = VectorParameter() {group=9_3_EarStripes, parameterName=Ear_Color3, defaultValue=<rgba 0 0.0636 0.3646 1>}
  n3 = Time()
  n4 = Abs(None=Sine(None=Multiply(A=@n3) {constB=2}) {period=10})
  n5 = Multiply(A=@n4) {constB=0.02}
  n6 = Multiply(A=@n4) {constB=0.05}
  n7 = Subtract(A=TextureCoordinate() {coordinateIndex=1}, B=AppendVector(A=Add(A=@n5) {constB=0.5}, B=Add(A=@n6) {constB=0.7}))
  n8 = Arctangent2!(Y=ComponentMask(None=@n7) {r=true}, X=ComponentMask(None=@n7) {g=true})
  n9 = Constant() {r=2}
  n10 = Frac(None=Divide(
    A=ComponentMask(None=@n8) {r=true}
    B=MaterialFunctionCall(Multiplier=@n9) {materialFunction=&Pi}.Result
   ))
  n11 = Time()
  n12 = ScalarParameter() {defaultValue=0.1, group=9_3_EarStripes, parameterName=Y Coord}
  n13 = Multiply(A=@n11, B=@n12)
  n14 = Multiply(A=@n7, B=@n7)
  n15 = SquareRoot(None=Add(A=ComponentMask(None=@n14) {r=true}, B=ComponentMask(None=@n14) {g=true}))
  n16 = ScalarParameter() {defaultValue=0.1, group=9_3_EarStripes, parameterName=X Coord}
  n17 = Sine(None=AppendVector(A=Add(A=@n10, B=@n13), B=Add(A=@n15, B=@n16)))
  n18 = TextureSampleParameter2D(UVs=MaterialFunctionCall(UVs=@n17, "Rotation Angle (0-1)"=Constant() {r=0.3}) {materialFunction=&CustomRotator}.Rotated Values) {group=8_InputTextureAni, parameterName=EarAni, samplerType=SAMPLERTYPE_Masks, texture=/VehicleCosmetics/Textures/T_Black_Masks}
  n19 = ScalarParameter() {group=9_3_EarStripes, parameterName=PosX_1}
  n20 = ScalarParameter() {group=9_3_EarStripes, parameterName=PosY_1}
  n21 = TextureSampleParameter2D(UVs=Subtract(A=TextureCoordinate() {coordinateIndex=1}, B=AppendVector(A=@n19, B=@n20))) {
   constCoordinate=1
   group=8_InputTextureAni
   parameterName=Eff_Ani
   samplerType=SAMPLERTYPE_Masks
   texture=&T_Black_Masks
  }
  n22 = Add(
   A=Multiply(A=@n21.B, B=ScalarParameter() {defaultValue=150, group=9_3_EarStripes, parameterName=Waves}) {constB=6}
   B=Multiply(A=Time(), B=ScalarParameter() {defaultValue=30, group=9_3_EarStripes, parameterName=TimeScale_1}) {constB=5}
  )
  n23 = Saturate(None=Add(A=@n18.R, B=Sine(None=@n22) {period=25}))
  n24 = VectorParameter() {group=9_3_EarStripes, parameterName=Ear_Color1, defaultValue=<rgba 0.7473 0.0596 0.8802 1>}
  n25 = TextureSampleParameter2D(UVs=MaterialFunctionCall(UVs=@n17, "Rotation Angle (0-1)"=Constant() {r=0.29}) {materialFunction=&CustomRotator}.Rotated Values) {group=8_InputTextureAni, parameterName=EarAni, samplerType=SAMPLERTYPE_Masks, texture=&T_Black_Masks}
  n26 = Saturate(None=Add(A=@n25.R, B=Cosine(None=@n22) {period=20}))
  n27 = ScalarParameter() {group=9_3_EarStripes, parameterName=PosX}
  n28 = ScalarParameter() {group=9_3_EarStripes, parameterName=PosY}
  n29 = TextureSampleParameter2D(UVs=Subtract(A=TextureCoordinate() {coordinateIndex=1}, B=AppendVector(A=@n27, B=@n28))) {
   constCoordinate=1
   group=8_InputTextureAni
   parameterName=Eff_Ani
   samplerType=SAMPLERTYPE_Masks
   texture=&T_Black_Masks
  }
  n30 = Add(
   A=Multiply(A=@n29.B, B=ScalarParameter() {defaultValue=150, group=9_3_EarStripes, parameterName=Waves}) {constB=6}
   B=Multiply(A=Time(), B=ScalarParameter() {defaultValue=30, group=9_3_EarStripes, parameterName=TimeScale}) {constB=5}
  )
  n31 = Saturate(None=Multiply(A=@n29.B, B=Sine(None=@n30) {period=24}) {constB=6})
  n32 = ScalarParameter() {defaultValue=4, group=9_3_EarStripes, parameterName=WaveWidth}
  n33 = Constant() {r=1}
  n34 = LinearInterpolate(
   A=LinearInterpolate(A=LinearInterpolate(A=@n1.RGB, B=@n2.RGB, Alpha=@n23), B=Multiply(A=@n24.RGB, B=@n26), Alpha=@n26)
   Alpha=If(A=Power(Base=@n31, Exp=@n32) {constExponent=4}, "A > B"=@n33, "A == B"=Constant(), "A < B"=@n33) {equalsThreshold=0}
  ) {constB=0}
  n35 = ScalarParameter() {defaultValue=0.5, group=9_3_EarStripes, parameterName=StrengthEmissive}
  FunctionOutput(None=SetMaterialAttributes!(MaterialAttributes, BaseColor=@n34, EmissiveColor=Multiply(A=@n34, B=@n35)) {attributeSetTypes [#1, #2]})
  Multiply(A=@n12) {constB=2}
  Multiply(A=Sine(None=@n11)) {constB=0}
  Multiply(A=@n11, B=@n16)
 ]}
