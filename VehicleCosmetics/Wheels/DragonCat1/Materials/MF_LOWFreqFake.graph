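# Material function for the DragonCat1 wheel cosmetic that generates a fake "low frequency" pulse value from time alone.
# Two sine waves with different periods, one on scaled time and one on offset time, are each divided and remapped with bias and scale.
# Their product is clamped below full intensity and output through a named reroute, giving an irregular looping modulation without any external input.
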MaterialFunction /VehicleCosmetics/Wheels/DragonCat1/Materials/MF_LOWFreqFake
graph {nodes [
  n1 = Time()
  n2 = Sine(None=Multiply(A=@n1) {constB=1.2}) {period=0.3}
  n3 = Sine(None=Add(A=@n1) {constB=0.3}) {period=0.4}
  n4 = Multiply(A=ConstantBiasScale(None=Divide(A=@n2)), B=ConstantBiasScale(None=Divide(A=@n3)) {bias=1.2, scale=1})
  FunctionOutput(None=NamedRerouteDeclaration(None=Clamp(None=@n4) {maxDefault=0.7}) {name=LOW Freq out Clamp, nodeColor=<rgba 1 0.5156 0 1>})
 ]}
