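# Material function for the DragonCat1 wheel's animated infinity-mirror emissive effect, pulsed by a fake low-frequency music signal.
# That signal drives the depth offset and gradient darkness of a dot-mask texture sampled through the shared infinity offset function on UV1.
# The red result tints a hue-shifted custom color, and green blends toward the unshifted, saturation-adjusted color.
# Opacity mask is a radial falloff from UV1 center.
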MaterialFunction /VehicleCosmetics/Wheels/DragonCat1/Materials/MF_InfinityMirror_MusicCreative
graph {
 properties {previewBlendMode=BLEND_Masked}
 nodes [
  n1 = TextureObjectParameter() {group=8_InputTextureAni, parameterName=Inifiny_Texture, texture=/VehicleCosmetics/Wheels/DragonCat1/Textures/T_Rim_DragonCat1_DotMask}
  n2 = ScalarParameter() {defaultValue=-1, group=9_2_Infinity_Ani, parameterName=Offset}
  n3 = Multiply(
   A=MaterialFunctionCall() {materialFunction=&/MF_LOWFreqFake}.Result
   B=ScalarParameter() {defaultValue=1, group=9_2_Infinity_Ani, parameterName=SpeedAnimated, sortPriority=33}
  )
  n4 = ScalarParameter() {defaultValue=0.01, group=9_2_Infinity_Ani, parameterName=GradientDarkness}
  n5 = Multiply(A=TextureCoordinate() {coordinateIndex=1}, B=ScalarParameter() {defaultValue=1, group=9_2_Infinity_Ani, parameterName=Tile})
  n6 = ScalarParameter() {group=9_2_Infinity_Ani, parameterName=Shift}
  n7 = Saturate(None=MaterialFunctionCall!(
    Texture=@n1
    Offset=Add(A=@n2, B=@n3)
    RotateUVs=StaticBool()
    GradientDarkness=Multiply(A=@n4, B=@n3)
    GradientFalloff
    UVs=Add(A=@n5, B=@n6)
    Opt_LowQualityOffset
    Opt_MediumQualityOffset
   ) {materialFunction=/VehicleCosmetics/SharedMaterials/MaterialFunctions/MF_Infinity_OffsetFunction}.Result)
  n8 = ComponentMask(None=@n7) {r=true}
  n9 = Desaturation(
   None=VectorParameter() {group=7_BaseSurAni, parameterName=CustomColor, defaultValue=<rgba 0.1437 0.8188 1 1>}.RGB
   Fraction=ScalarParameter() {defaultValue=-0.5, group=9_2_Infinity_Ani, parameterName=DesaturateColor}
  ) {luminanceFactors=<rgba 0.3 0.59 0.11 0>}
  n10 = MaterialFunctionCall(Hue Shift Percentage=ScalarParameter() {defaultValue=0.2, group=9_2_Infinity_Ani, parameterName=Angle}, Texture=@n9) {materialFunction=&HueShift}
  n11 = ComponentMask(None=@n7) {g=true}
  n12 = Add(A=TextureCoordinate() {coordinateIndex=1}) {constB=-0.5}
  n13 = DotProduct(A=@n12, B=@n12)
  n14 = SetMaterialAttributes!(
   MaterialAttributes
   Specular=Constant()
   Roughness=Constant() {r=1}
   EmissiveColor=LinearInterpolate(A=Multiply(A=@n8, B=@n10.Result), B=@n9, Alpha=Add(B=@n11) {constA=0.01})
   OpacityMask=Saturate(None=OneMinus(None=@n13))
  ) {attributeSetTypes [
    #1
    #2
    #3
    #4
   ]}
  FunctionOutput(None=@n14)
 ]
}
