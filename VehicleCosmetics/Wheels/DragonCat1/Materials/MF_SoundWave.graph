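# Material function for the DragonCat1 wheel cosmetic that outputs an animated, horizontal sound-wave line mask over UV space.
# A sine scrolling along U by Speed and Width is pulsed by a time sine and scaled by Strong_Wave and MF_LOWFreqFake, apparently a faked low-frequency audio signal.
# That wave, scaled by Height, offsets V, and the clamped distance from the center is shaped by SmoothStep with adjustable Min and Max.
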MaterialFunction /VehicleCosmetics/Wheels/DragonCat1/Materials/MF_SoundWave
graph {nodes [
  n1 = FunctionInput() {bUsePreviewValueAsDefault=true, inputName=SmoothStep_Min, inputType=FunctionInput_Scalar, sortPriority=5, previewValue=<0.15 0 0 1>}
  n2 = FunctionInput(Preview=TextureCoordinate()) {bUsePreviewValueAsDefault=true, inputName=UV, inputType=FunctionInput_Vector2, sortPriority=4, previewValue=<1 0 0 1>}
  n3 = MaterialFunctionCall(Float2=@n2) {materialFunction=&BreakOutFloat2Components}
  n4 = FunctionInput() {bUsePreviewValueAsDefault=true, inputName=Strong_Wave, inputType=FunctionInput_Scalar, previewValue=<1.5 0 0 1>}
  n5 = Time()
  n6 = FunctionInput() {bUsePreviewValueAsDefault=true, inputName=Speed, inputType=FunctionInput_Scalar, sortPriority=1, previewValue=<1 0 0 1>}
  n7 = FunctionInput() {bUsePreviewValueAsDefault=true, inputName=Width, inputType=FunctionInput_Scalar, sortPriority=2, previewValue=<2 0 0 1>}
  n8 = Add(A=Multiply(A=@n5, B=@n6), B=Multiply(A=@n3.R, B=@n7))
  n9 = Multiply(
   A=Multiply(A=MaterialFunctionCall() {materialFunction=&/MF_LOWFreqFake}.Result, B=@n4)
   B=Multiply(A=Sine(None=@n5), B=Sine(None=@n8))
  )
  n10 = FunctionInput() {bUsePreviewValueAsDefault=true, inputName=Height, inputType=FunctionInput_Scalar, sortPriority=3, previewValue=<0.15 0 0 1>}
  n11 = Distance(A=Add(A=@n3.G, B=Multiply(A=@n9, B=@n10)), B=Constant() {r=0.5})
  FunctionOutput(None=SmoothStep(
    Min=@n1
    Max=FunctionInput() {bUsePreviewValueAsDefault=true, inputName=SmoothStep_Max, inputType=FunctionInput_Scalar, sortPriority=6}
    Value=Clamp(None=@n11)
   ))
 ]}
