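# Material function for the DragonCat1 tire cosmetic, adding animated, fake audio-reactive sound-wave bands.
# Three tinted MF_SoundWave layers take their height from MF_LOWFreqFake; the third blends in where the first two are absent.
# Texture channels mask the waves and tint base, custom and accent colors, with a BumpOffset driven by the fake signal shifting the dot accents.
# The result feeds both BaseColor and Emissive, alongside specular, roughness and normal.
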MaterialFunction /VehicleCosmetics/Wheels/DragonCat1/Materials/MF_Tire_DragonCat1
graph {nodes [
  n1 = TextureSampleParameter2D() {group=8_InputTextureAni, parameterName=Tire_RGB_Ani, texture=/VehicleCosmetics/Wheels/DragonCat1/Textures/T_Tire_DragonCat1_RGB}
  n2 = ScalarParameter() {defaultValue=2, group=9_1_TireWave1, parameterName=StrengthWave1}
  n3 = ScalarParameter() {defaultValue=5, group=9_1_TireWave1, parameterName=Soundmultiplier1}
  n4 = MaterialFunctionCall() {materialFunction=&/MF_LOWFreqFake}
  n5 = ScalarParameter() {defaultValue=1.3, group=9_1_TireWave1, parameterName=PosX1}
  n6 = ScalarParameter() {defaultValue=1.05, group=9_1_TireWave1, parameterName=PosY1}
  n7 = ScalarParameter() {defaultValue=0.3, group=9_1_TireWave1, parameterName=Zoom_Wave1}
  n8 = Divide(A=Multiply(A=TextureCoordinate() {uTiling=1.2, vTiling=1.05}, B=AppendVector(A=@n5, B=@n6)), B=@n7)
  n9 = Divide(B=@n7)
  n10 = MaterialFunctionCall!(
   Strong_Wave
   Speed=ScalarParameter() {defaultValue=1, group=9_1_TireWave1, parameterName=Sine_Speed1}
   Width=ScalarParameter() {defaultValue=2, group=9_1_TireWave1, parameterName=Sine_Width1}
   Height=Clamp(None=Multiply(A=@n3, B=@n4.Result)) {maxDefault=0.7, minDefault=0.1}
   UV=Subtract(A=Add(A=@n8) {constB=0.5}, B=Divide(A=@n9))
   SmoothStep_Min=ScalarParameter() {defaultValue=0.2, group=9_1_TireWave1, parameterName=Sine_SS_Min1}
   SmoothStep_Max=ScalarParameter() {group=9_1_TireWave1, parameterName=Sine_SS_Max1}
  ) {materialFunction=&/MF_SoundWave}
  n11 = Multiply(A=@n10.Result, B=VectorParameter() {group=9_1_TireWave1, parameterName=Color_Wave1, defaultValue=<rgba 0.008 0.6 0.4 1>}.RGB)
  n12 = ScalarParameter() {defaultValue=2, group=9_1_TireWave2, parameterName=StrengthWave2}
  n13 = ScalarParameter() {defaultValue=0.2, group=9_1_TireWave2, parameterName=Soundmultiplier2}
  n14 = MaterialFunctionCall() {materialFunction=&MF_LOWFreqFake}
  n15 = ScalarParameter() {defaultValue=1.2, group=9_1_TireWave2, parameterName=PosX2}
  n16 = ScalarParameter() {defaultValue=0.95, group=9_1_TireWave2, parameterName=PosY2}
  n17 = ScalarParameter() {defaultValue=0.3, group=9_1_TireWave2, parameterName=Zoom_Wave2}
  n18 = Divide(A=Multiply(A=TextureCoordinate(), B=AppendVector(A=@n15, B=@n16)), B=@n17)
  n19 = Divide(B=@n17)
  n20 = MaterialFunctionCall(
   Speed=ScalarParameter() {defaultValue=1, group=9_1_TireWave2, parameterName=Sine_Speed2}
   Width=ScalarParameter() {defaultValue=2, group=9_1_TireWave2, parameterName=Sine_Width2}
   Height=Clamp(None=Multiply(A=@n13, B=@n14.Result)) {minDefault=0.1}
   UV=Subtract(A=Add(A=@n18) {constB=0.5}, B=Divide(A=@n19))
   SmoothStep_Min=ScalarParameter() {defaultValue=0.08, group=9_1_TireWave2, parameterName=Sine_SS_Min2}
   SmoothStep_Max=ScalarParameter() {group=9_1_TireWave2, parameterName=Sine_SS_Max2}
  ) {materialFunction=&MF_SoundWave}
  n21 = Multiply(A=@n20.Result, B=VectorParameter() {group=9_1_TireWave2, parameterName=Color_Wave2, defaultValue=<rgba 0.0033 0.0143 0.194 1>}.RGB)
  n22 = ScalarParameter() {defaultValue=0.2, group=9_1_TireWave3, parameterName=Soundmultiplier3}
  n23 = MaterialFunctionCall() {materialFunction=&MF_LOWFreqFake}
  n24 = ScalarParameter() {defaultValue=1.2, group=9_1_TireWave3, parameterName=PosX3}
  n25 = ScalarParameter() {defaultValue=1, group=9_1_TireWave3, parameterName=PosY3}
  n26 = ScalarParameter() {defaultValue=0.3, group=9_1_TireWave3, parameterName=Zoom_Wave3}
  n27 = Divide(A=Multiply(A=TextureCoordinate() {uTiling=1.2, vTiling=1.05}, B=AppendVector(A=@n24, B=@n25)), B=@n26)
  n28 = Divide(B=@n26)
  n29 = MaterialFunctionCall(
   Speed=ScalarParameter() {defaultValue=1, group=9_1_TireWave3, parameterName=Sine_Speed3}
   Width=ScalarParameter() {defaultValue=2, group=9_1_TireWave3, parameterName=Sine_Width3}
   Height=Clamp(None=Multiply(A=@n22, B=@n23.Result)) {maxDefault=0.7, minDefault=0.1}
   UV=Subtract(A=Add(A=@n27) {constB=0.5}, B=Divide(A=@n28))
   SmoothStep_Min=ScalarParameter() {defaultValue=0.2, group=9_1_TireWave3, parameterName=Sine_SS_Min3}
   SmoothStep_Max=ScalarParameter() {group=9_1_TireWave3, parameterName=Sine_SS_Max3}
  ) {materialFunction=&MF_SoundWave}
  n30 = VectorParameter() {group=9_1_TireWave3, parameterName=Color_Wave3, defaultValue=<rgba 0.005 0.0658 0.9075 1>}
  n31 = LinearInterpolate(
   A=Add(A=Multiply(A=@n2, B=@n11), B=Multiply(A=@n12, B=@n21))
   B=Multiply(A=ScalarParameter() {defaultValue=2, group=9_1_TireWave3, parameterName=StrengthWave3}, B=Multiply(A=@n29.Result, B=@n30.RGB))
   Alpha=OneMinus(None=Add(A=@n10.Result, B=@n20.Result))
  )
  n32 = LinearInterpolate(
   A=LinearInterpolate(
    A=VectorParameter() {group=9_0_RimSurAni, parameterName=BaseColorTire, defaultValue=<rgba 0.02 0.02 0.02 1>}.RGB
    B=VectorParameter() {group=9_0_RimSurAni, parameterName=CustomColor, defaultValue=<rgba 0.1437 0.8188 1 1>}.RGB
    Alpha=@n1.R
   )
   B=VectorParameter() {group=9_0_RimSurAni, parameterName=AccentColor, defaultValue=<rgba 0 0 1 1>}.RGB
   Alpha=@n1.G
  )
  n33 = VectorParameter() {group=8_InputTextureAni, parameterName=DotAccentColor, defaultValue=<rgba 0.3457 0.632 0.5204 1>}
  n34 = TextureSampleParameter2D(UVs=BumpOffset(Height=Subtract(B=@n4.Result))) {group=8_InputTextureAni, parameterName=Tire_RGB_Ani, texture=&T_Tire_DragonCat1_RGB}
  n35 = Add(
   A=Multiply(A=@n1.A, B=LinearInterpolate(B=@n31, Alpha=@n1.A))
   B=Multiply(
    A=ScalarParameter() {defaultValue=0.2, group=9_1_TireAniProperty, parameterName=StrengthBase}
    B=LinearInterpolate(A=@n32, B=@n33.RGB, Alpha=@n34.B)
   )
  )
  n36 = TextureSampleParameter2D() {group=1_InputTextures, parameterName=Tire_Normal, samplerType=SAMPLERTYPE_Normal, texture=/VehicleCosmetics/Textures/T_BlankNormal_B}
  n37 = SetMaterialAttributes!(
   MaterialAttributes
   BaseColor=@n35
   Specular=ScalarParameter() {defaultValue=0.7, group=9_1_TireAniProperty, parameterName=TireSpec}
   Roughness=ScalarParameter() {defaultValue=0.6, group=9_1_TireAniProperty, parameterName=TireRoughness}
   EmissiveColor=@n35
   Normal=@n36.RGB
  ) {attributeSetTypes [
    #1
    #2
    #3
    #4
    #5
   ]}
  FunctionOutput(None=@n37)
 ]}
