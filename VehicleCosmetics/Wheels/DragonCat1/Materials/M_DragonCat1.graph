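# Masked vehicle wheel material for DragonCat1, layering animated rim effects and a tire over the simple wheel parent function.
# A time-based sine pulse, scaled by MusicEmissivePower and an inverted Eff_Ani channel, tints base color into emissive, lerped with parent emissive by rim masks.
# Eff_Ani channels blend in ear-stripe and infinity-mirror functions; vertex red blends in the tire.
# The result passes through shared damage and de-rez transition functions.
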Material /VehicleCosmetics/Wheels/DragonCat1/Materials/M_DragonCat1
graph {
 properties {
  bAutomaticallySetUsageInEditor=false
  bUseMaterialAttributes=true
  bUsedWithMeshDeformer=true
  bUsedWithSkeletalMesh=true
  blendMode=BLEND_Masked
  opacityMaskClipValue=0.75
  usedShadingModels=Default Lit
 }
 nodes [
  n1 = MaterialFunctionCall() {materialFunction=&MF_WheelParent_Simple}
  n2 = GetMaterialAttributes(None=@n1.MaterialAttributes) {attributeGetTypes [#1, #2]}
  n3 = Sine(None=Multiply(B=Time()) {constA=0.2})
  n4 = TextureSampleParameter2D() {
   group=8_InputTextureAni
   parameterName=Eff_Ani
   samplerType=SAMPLERTYPE_Masks
   texture=/VehicleCosmetics/Textures/T_Black_Masks
   channelNames {a="", b=Effect in Ear, g=EarStrip, r=Infinity Dot}
  }
  n5 = Multiply(A=Multiply(A=Add(B=@n3) {constA=1}) {constA=0.2, constB=0.4}, B=OneMinus(None=@n4.A)) {constA=0.2, constB=0.4}
  n6 = ScalarParameter() {defaultValue=2, group=9_0_RimSurAni, parameterName=MusicEmissivePower}
  n7 = MaterialFunctionCall(Float3=@n1.RimRGBAMask) {materialFunction=&BreakOutFloat3Components}
  n8 = LinearInterpolate(A=Multiply(A=@n2.Base Color, B=Multiply(A=@n5, B=@n6)), B=@n2.Emissive Color, Alpha=Add(A=Add(A=@n7.R, B=@n7.R), B=@n7.B)) {constB=0}
  n9 = TextureSampleParameter2D() {
   group=8_InputTextureAni
   parameterName=Eff_Ani
   samplerType=SAMPLERTYPE_Masks
   texture=&T_Black_Masks
   channelNames {a="", b=Effect in Ear, g=EarStrip, r=Infinity Dot}
  }
  n10 = BlendMaterialAttributes(
   A=BlendMaterialAttributes(
    A=SetMaterialAttributes(MaterialAttributes=@n2.MaterialAttributes, Emissive Color=@n8) {attributeSetTypes [#2]}
    B=MaterialFunctionCall() {materialFunction=&/MF_DragonCat1_EarStripes}.Result
    Alpha=@n9.EarStrip
   )
   B=MaterialFunctionCall() {materialFunction=&/MF_InfinityMirror_MusicCreative}.Result
   Alpha=@n9.Infinity Dot
  )
  n11 = MaterialFunctionCall() {materialFunction=&/MF_Tire_DragonCat1}
  n12 = VertexColor()
  MaterialFunctionCall(In=MaterialFunctionCall(Material=BlendMaterialAttributes(A=@n10, B=@n11.Result, Alpha=@n12.R)) {materialFunction=&MF_BaseVehicle_Damage}.Result) {materialFunction=&MF_VehicleDeRez_Transition}
 ]
}
